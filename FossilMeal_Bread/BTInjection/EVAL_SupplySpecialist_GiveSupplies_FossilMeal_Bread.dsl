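; Behavior tree evaluator for the FossilMeal Bread NPC boss that, going by its name, supports the supply specialist's give-supplies behavior.
; On enter, it runs the parent logic, casts the pawn to the boss class and binds Hired to the convert component's OnConverted event.
; Hired caches the converter pawn and that player's Athena player state in variables.
; The supply-giving logic itself is not in this graph.
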(event EventOnEnter
  (params (OwnerComp :type Behavior%20Tree%20Component%20Object%20Reference))
  (|CallParentFunction OwnerComp)
  (bind _asbp_player_pawn_npcboss_fossil_meal_bread (Utilities|Casting|CastToBP_PlayerPawn_NPCBoss_FossilMeal_Bread (Variables|Default|GetPawn)))
  (ConvertEvents|BindEventtoOnConvertedEvent (Variables|/Script/FortniteGame.FortPlayerPawn|GetConvertComponent _asbp_player_pawn_npcboss_fossil_meal_bread) (delegate Custom|Hired)))

(event Custom|Hired
  (params (InstigatorPawn :type Fort%20Pawn%20Object%20Reference) (ConvertedPawn :type Fort%20Pawn%20Object%20Reference))
  (bind _output_get (Variables|Default|SetConverterPawn (Convert|GetConverterPawn (Variables|/Script/FortniteGame.FortPlayerPawn|GetConvertComponent (stale EventOnEnter _asbp_player_pawn_npcboss_fossil_meal_bread)))))
  (bind _asfort_player_state_athena (Utilities|Casting|CastToFortPlayerStateAthena (Variables|/Script/Engine.Controller|GetPlayerState (Pawn|GetController _output_get))))
  (Variables|Default|SetConverterPlayerStateAthena _asfort_player_state_athena))
