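; Behavior tree injection evaluator that decides whether the FossilMeal Bread NPC boss may start its emergency leap.
; EvaluateStartingConditions passes only if the boss or its hiree is at critical health and the boss's EmergencyLeptOnce flag is unset.
; The boss must also be farther than the scalable MaxEmergencyDistance_Hf value from its hiree, falling back to its spawn location when no valid hiree exists.
; The parent evaluation is called, but its result is not used.
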(fn EvaluateStartingConditions
  (params (OwnerComp :type Behavior%20Tree%20Component%20Object%20Reference))
  (bind _bool (CallFunction|IsBossorHireeatCriticalHealth))
  (bind _breadboss (Variables|Default|GetBreadBoss))
  (bind _emergencyleptonce (Variables|/FossilMeal_Bread/AI/Pawns/BP_PlayerPawn_NPCBoss_FossilMeal_Bread.BP_PlayerPawn_NPCBoss_FossilMeal_Bread_C|GetEmergencyLeptOnce _breadboss))
  (bind _returnvalue (not _emergencyleptonce))
  (|CallParentFunction OwnerComp)
  (Utilities|IsValid (Variables|/FossilMeal_Bread/AI/Pawns/BP_PlayerPawn_NPCBoss_FossilMeal_Bread.BP_PlayerPawn_NPCBoss_FossilMeal_Bread_C|GetHireeRef _breadboss)
    (:Is%20Valid
      (return (Math|Boolean|ANDBoolean (> (Math|Vector|Distance(Vector) (Transformation|GetActorLocation (Variables|Default|GetBreadBoss)) (Transformation|GetActorLocation (Variables|/FossilMeal_Bread/AI/Pawns/BP_PlayerPawn_NPCBoss_FossilMeal_Bread.BP_PlayerPawn_NPCBoss_FossilMeal_Bread_C|GetHireeRef (Variables|Default|GetBreadBoss)))) (ScalableFloat|GetValueatLevel (Variables|Default|GetMaxEmergencyDistance_Hf))) _bool _returnvalue)))
    (:Is%20Not%20Valid
      (return (Math|Boolean|ANDBoolean _bool _returnvalue (> (Math|Vector|Distance(Vector) (Transformation|GetActorLocation (Variables|Default|GetBreadBoss)) (Variables|/FossilMeal_Bread/AI/Pawns/BP_PlayerPawn_NPCBoss_FossilMeal_Bread.BP_PlayerPawn_NPCBoss_FossilMeal_Bread_C|GetSpawnLocation (Variables|Default|GetBreadBoss))) (ScalableFloat|GetValueatLevel (Variables|Default|GetMaxEmergencyDistance_Hf))))))))
