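; Behavior tree evaluator that decides when the FossilMeal Bread NPC boss may start its leap ability.
; On enter it caches the AI controller and boss pawn.
; Starting requires a valid blackboard target at or beyond a minimum distance, the ability enabled, leaping allowed and not underway, and neither boss nor hiree at critical combined health-plus-shield.
; Ongoing conditions always return true, and tuning comes from scalable floats.
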(event EventOnEnter
  (params (OwnerComp :type Behavior%20Tree%20Component%20Object%20Reference))
  (bind _asfort_athena_aibot_controller (Utilities|Casting|CastToFortAthenaAIBotController (Components|GetOwner OwnerComp)))
  (bind _output_get (Variables|Default|SetBossController _asfort_athena_aibot_controller))
  (bind _asbp_player_pawn_npcboss_fossil_meal_bread (Utilities|Casting|CastToBP_PlayerPawn_NPCBoss_FossilMeal_Bread (Pawn|GetControlledPawn _output_get)))
  (Variables|Default|SetBreadBoss _asbp_player_pawn_npcboss_fossil_meal_bread))

(fn EvaluateOngoingConditions
  (params (OwnerComp :type Behavior%20Tree%20Component%20Object%20Reference))
  (|CallParentFunction OwnerComp)
  (return true))

(fn EvaluateStartingConditions
  (params (OwnerComp :type Behavior%20Tree%20Component%20Object%20Reference))
  (bind _breadboss (Variables|Default|GetBreadBoss))
  (|CallParentFunction OwnerComp)
  (bind _targetactor (Class|BPFLAICommonFunctions|GetTargetActorfromBlackboard _breadboss))
  (Utilities|IsValid _targetactor
    (:Is%20Valid
      (return (Math|Boolean|ANDBoolean (ScalableFloat|GetValueasBool (Variables|Default|GetCanUseAbility_Hf)) (>= (Transformation|GetDistanceTo _breadboss _targetactor) (ScalableFloat|GetValueatLevel (Variables|Default|GetMinDistanceToTarget_Hf))) (Variables|/FossilMeal_Bread/AI/Pawns/BP_PlayerPawn_NPCBoss_FossilMeal_Bread.BP_PlayerPawn_NPCBoss_FossilMeal_Bread_C|GetCanLeap (Variables|Default|GetBreadBoss)) (not (Variables|/FossilMeal_Bread/AI/Pawns/BP_PlayerPawn_NPCBoss_FossilMeal_Bread.BP_PlayerPawn_NPCBoss_FossilMeal_Bread_C|GetIsLeaping (Variables|Default|GetBreadBoss))) (not (CallFunction|IsBossorHireeatCriticalHealth)))))
    (:Is%20Not%20Valid
      (return false))))

(fn IsBossOrHireeAtCriticalHealth
  (bind _criticalhealthpercent_hf (Variables|Default|GetCriticalHealthPercent_Hf))
  (bind _returnvalue (ScalableFloat|GetValueatLevel _criticalhealthpercent_hf))
  (bind _maxeffectivehealth (LocalVariables|GetMaxEffectiveHealth))
  (bind _currenteffectivehealth (LocalVariables|GetCurrentEffectiveHealth))
  (bind _returnvalue_2 (/ _currenteffectivehealth _maxeffectivehealth))
  (bind _returnvalue_3 (<= _returnvalue_2 _returnvalue))
  (bind _breadboss (Variables|Default|GetBreadBoss))
  (bind _breadboss_2 (Variables|Default|GetBreadBoss))
  (bind _hireeref (Variables|/FossilMeal_Bread/AI/Pawns/BP_PlayerPawn_NPCBoss_FossilMeal_Bread.BP_PlayerPawn_NPCBoss_FossilMeal_Bread_C|GetHireeRef _breadboss_2))
  (LocalVariables|SetMaxEffectiveHealth (+ (FortPawn|GetMaxHealth _breadboss) (FortPawn|GetMaxShield _breadboss)))
  (LocalVariables|SetCurrentEffectiveHealth (+ (FortPawn|GetHealth _breadboss) (FortPawn|GetShield _breadboss)))
  (Utilities|IsValid _hireeref
    (:Is%20Valid
      (LocalVariables|SetHireeMaxEffectiveHealth (+ (FortPawn|GetMaxHealth _hireeref) (FortPawn|GetMaxShield _hireeref)))
      (LocalVariables|SetHireeCurrentEffectiveHealth (+ (FortPawn|GetHealth _hireeref) (FortPawn|GetShield _hireeref)))
      (return (or (<= (/ (LocalVariables|GetHireeCurrentEffectiveHealth) (LocalVariables|GetHireeMaxEffectiveHealth)) (ScalableFloat|GetValueatLevel (Variables|Default|GetCriticalHealthPercent_Hf))) _returnvalue_3)))
    (:Is%20Not%20Valid
      (return _returnvalue_3))))
